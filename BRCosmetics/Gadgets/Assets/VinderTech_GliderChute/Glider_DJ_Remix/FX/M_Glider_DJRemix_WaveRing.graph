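# Additive unlit mesh-particle material for the DJ Remix glider's wave ring effect, drawing a glowing, pulsing, distorted band.
# Vertex offset pushes outward from particle center using two panning noise textures, scaled by sine pulses or a dynamic parameter via static switch, plus extra sine wobble.
# A soft UV band masks time-blended orange-purple, hue-shifted near the center and oversaturated through negative desaturation.
# HDR remapping and particle color finish emissive; depth fade softens alpha.
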Material /BRCosmetics/Gadgets/Assets/VinderTech_GliderChute/Glider_DJ_Remix/FX/M_Glider_DJRemix_WaveRing
graph {
 properties {
  bUsedWithMeshParticles=true
  blendMode=BLEND_Additive
  pixelDepthOffsetMode=PDOM_Legacy
  shadingModel=MSM_Unlit
  twoSided=true
  usedShadingModels=Unlit
 }
 nodes [
  n1 = TextureCoordinate()
  n2 = AppendVector(
   A=Multiply(A=ScalarParameter() {defaultValue=3, parameterName=WPO_U}, B=ComponentMask(None=@n1) {r=true})
   B=Multiply(A=ComponentMask(None=@n1) {g=true}, B=ScalarParameter() {defaultValue=3, parameterName=WPO_V})
  )
  n3 = Add(A=Time(), B=DynamicParameter() {defaultValue=<rgba 1 1 0 1>}.Param1)
  n4 = TextureCoordinate()
  n5 = AppendVector(
   A=Multiply(A=ScalarParameter() {defaultValue=2, parameterName=World position mask_U}, B=ComponentMask(None=@n4) {r=true})
   B=Multiply(A=ComponentMask(None=@n4) {g=true}, B=ScalarParameter() {defaultValue=2, parameterName=World position mask_V})
  )
  n6 = Multiply(
   A=TextureSample(UVs=Panner(Coordinate=@n2, Time=@n3) {speedX=0.2, speedY=0.55}) {texture=&T_Inky_Smoke_Tile}.RGB
   B=TextureSample(UVs=Panner(Coordinate=@n5) {speedX=-0.14, speedY=-0.4}) {texture=/Game/Global/Textures/Default/Tileable_Patterns/T_Tornado_Tile}.RGB
  )
  n7 = Normalize(VectorInput=Subtract(A=WorldPosition().XYZ, B=ParticlePositionWS()))
  n8 = Constant() {r=9}
  n9 = DynamicParameter() {defaultValue=<rgba 1 1 0 1>}
  n10 = Add(A=Multiply(A=Time()) {constB=1.8}, B=@n9.Param3)
  n11 = Multiply(A=ScalarParameter() {defaultValue=19, parameterName=Param_1}, B=Ceil(None=Sine(None=@n10)))
  n12 = Constant() {r=22}
  n13 = Time()
  n14 = Sine(None=Add(A=Multiply(A=@n13) {constB=2}, B=@n9.Param3))
  n15 = Clamp(None=Add(A=Add(A=@n8, B=@n11), B=Multiply(A=@n12, B=@n14))) {maxDefault=35}
  n16 = DynamicParameter() {defaultValue=<rgba 1 1 0 1>}
  n17 = Subtract(A=WorldPosition().XYZ, B=ParticlePositionWS())
  n18 = Multiply(A=Time()) {constB=0.6}
  n19 = Sine(None=Multiply(A=Add(A=@n9.Param3, B=@n18)) {constB=5})
  n20 = Multiply(A=Time()) {constB=0.64}
  n21 = Sine(None=Multiply(A=Add(A=@n9.Param3, B=@n20)) {constB=3})
  n22 = Add(A=Multiply(A=Constant() {r=4.2}, B=Add(A=@n19)), B=Ceil(None=Add(A=@n21)))
  n23 = DynamicParameter() {defaultValue=<rgba 1 1 0 1>}
  Add(
   A=Multiply(A=Multiply(A=@n6, B=@n7), B=StaticSwitchParameter(True=@n15, False=@n16.Param2)) {constB=32}
   B=Multiply(A=Normalize(VectorInput=@n17), B=Multiply(A=@n22, B=@n23.Param4))
  )
  n24 = ParticleColor()
  DepthFade(Opacity=@n24.A)
  n25 = TextureCoordinate()
  n26 = MaterialFunctionCall(
   Hue Shift Percentage=SphereMask(A=WorldPosition().XYZ, B=ParticlePositionWS(), Radius=ScalarParameter() {defaultValue=32, parameterName=HueShift_Size}) {attenuationRadius=50, hardnessPercent=10}
   Texture=LinearInterpolate(
    A=Constant3Vector() {constant=<rgba 0.8958 0.5196 0 0>}
    B=Constant3Vector() {constant=<rgba 0.7498 0.0778 1 0>}
    Alpha=MaterialFunctionCall() {materialFunction=&DebugTimeSine}.Result
   )
  ) {materialFunction=&HueShift}
  n27 = Constant() {r=-3}
  n28 = Multiply(
   A=SphereMask(A=ComponentMask(None=@n25) {g=true}, B=Constant() {r=0.5}, Radius=ScalarParameter() {defaultValue=0.35, parameterName=Size}) {attenuationRadius=0.35, hardnessPercent=0}
   B=Multiply(A=Desaturation(None=@n26.Result, Fraction=@n27))
  )
  n29 = Clamp(None=@n28)
  MaterialFunctionCall(Emissive=Multiply(
    A=MaterialFunctionCall(LDR Input=@n29, HDR Tint=@n28) {materialFunction=&DeriveHDRfromLDR}.RemappedRange
    B=@n24.RGB
   )) {materialFunction=&MF_FX_EmissiveAdjustments}
 ]
}
